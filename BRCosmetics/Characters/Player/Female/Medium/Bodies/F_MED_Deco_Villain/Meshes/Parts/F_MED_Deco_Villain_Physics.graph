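# Physics asset for the F_MED_Deco_Villain female body part, set up for secondary motion on a dangling helmet piece.
# Capsule bodies cover the head and a three-bone dyn_helmet_01 chain.
# Constraints link each chain bone to its parent, starting at the head, with tight, uneven swing and limited twist ranges that restrict how far the piece bends.
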PhysicsAsset /BRCosmetics/Characters/Player/Female/Medium/Bodies/F_MED_Deco_Villain/Meshes/Parts/F_MED_Deco_Villain_Physics
bodies {
 dyn_helmet_01 [{shapeType=Capsule, length=4.9972, radius=4.1564, center=<-0.7058 -0.119 -3.5062>, rotation=<rot 0 73.1443 -173.2243>}]
 dyn_helmet_01_a [{shapeType=Capsule, length=7.1039, radius=4.0346, center=<0.0002 0 -4.432>}]
 dyn_helmet_01_b [{shapeType=Capsule, length=4.0985, radius=3.8863, center=<-0.0001 -1.1053 -1.0597>, rotation=<rot 0 0 37.8459>}]
 head [{shapeType=Capsule, length=15.7469, radius=10.222, center=<3.536 2.4888 0>, rotation=<rot 90 0 -8.9997>}]
}
constraints [
 {bone1Name=dyn_helmet_01, bone2Name=head, swing1LimitDegrees=8, swing2LimitDegrees=16, twistLimitDegrees=16}
 {bone1Name=dyn_helmet_01_a, bone2Name=dyn_helmet_01, swing1LimitDegrees=8, swing2LimitDegrees=16, twistLimitDegrees=16}
 {bone1Name=dyn_helmet_01_b, bone2Name=dyn_helmet_01_a, swing1LimitDegrees=8, swing2LimitDegrees=16, twistLimitDegrees=16}
]
